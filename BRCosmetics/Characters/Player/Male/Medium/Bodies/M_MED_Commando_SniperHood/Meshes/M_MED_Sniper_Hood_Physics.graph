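# Physics asset for the male medium Commando Sniper Hood body, defining secondary motion for its shawl and back flap bones.
# Spheres and capsules on the spine and right arm form body volumes, including named chest and rib flap colliders.
# Back flaps hang from spine_02, the front shawl from spine_05, and shoulder and arm chains from small anchor spheres on upperarm_r.
# Cloth-piece joints use mostly tight swing and twist limits, while anchor and most body joints rotate freely.
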PhysicsAsset /BRCosmetics/Characters/Player/Male/Medium/Bodies/M_MED_Commando_SniperHood/Meshes/M_MED_Sniper_Hood_Physics
bodies {
 dyn_backFlapEnd_01 [{shapeType=Box, extentX=0.8399, extentY=1.9229, extentZ=2.1076, center=<1.327 0.0002 -0.0305>, rotation=<rot 0 90 90>}]
 dyn_backFlapMain [{shapeType=Box, extentX=0.9849, extentY=8.6121, extentZ=5.295, center=<5.3093 0.9547 -0.0021>, rotation=<rot 0 -79.087 0>}]
 dyn_backFlap_01 [{shapeType=Box, extentX=1.51, extentY=1.76, extentZ=2.51, center=<1.3706 0.0002 -0.0018>, rotation=<rot 0 90 90>}]
 dyn_backFlap_02 [{shapeType=Box, extentX=1.453, extentY=3.9517, extentZ=5.8198, center=<2.7357 -0.4121 -0.1019>, rotation=<rot 0 90 90>}]
 dyn_backFlap_03 [{shapeType=Box, extentX=0.8062, extentY=2.0433, extentZ=2.9262, center=<1.8003 0.0233 -0.0946>, rotation=<rot 0 90 90>}]
 dyn_shawlBackA_01 [{shapeType=Capsule, length=3.2548, radius=1.2478, center=<2.5906 -0.5072 0.0056>, rotation=<rot -16.9415 -96.2814 92.8337>}]
 dyn_shawlBackA_02 [{shapeType=Box, extentX=1.2827, extentY=4.296, extentZ=2.8841, center=<4.2975 -0.0001 0.9304>, rotation=<rot 0 89.9999 17.3362>}]
 dyn_shawlBackA_03 [{shapeType=Box, extentX=1.3024, extentY=6.3297, extentZ=3.7009, center=<3.0063 0.0001 0.5513>, rotation=<rot 0 89.9999 11.4895>}]
 dyn_shawlFront_01 [{shapeType=Box, extentX=14.3999, extentY=1.3335, extentZ=5.4622, center=<7.8157 0.0002 -0.0001>, rotation=<rot -0.0001 0 15.2644>}]
 dyn_shawlShoulderBase [{shapeType=Sphere, radius=0.505}]
 dyn_shawlShoulderEnd_02 [{shapeType=Box, extentX=1.01, extentY=8.51, extentZ=4.01, center=<2.4911 0.0001 -0.8205>, rotation=<rot 0 -90 0>}]
 dyn_shawlShoulderFrontBase [{shapeType=Sphere, radius=0.505}]
 dyn_shawlShoulderFront_01 [{shapeType=Box, extentX=1.4511, extentY=6.5275, extentZ=4.6755, center=<2.0468 -0.0497 -0.2823>, rotation=<rot 7.8933 -89.9999 -2.1027>}]
 dyn_shawlShoulderFront_02 [{shapeType=Box, extentX=0.8179, extentY=4.7448, extentZ=3.2599, center=<1.5409 -0.0819 -0.292>, rotation=<rot 5.3954 -89.9999 0.3703>}]
 dyn_shawlShoulder_01 [{shapeType=Box, extentX=2.2894, extentY=11.6176, extentZ=7.5721, center=<6.0414 -0.5024 0.9905>, rotation=<rot 0 -90 0>}]
 dyn_shawlShoulder_02 [{shapeType=Box, extentX=4.6992, extentY=2.5604, extentZ=6.8606, center=<2.1414 -0.4247 0.5191>, rotation=<rot 90 0 0>}]
 dyn_shawl_ArmBase [{shapeType=Sphere, radius=0.505}]
 lowerarm_r [{shapeType=Capsule, length=4.2405, radius=5.6151, center=<-5.4562 -1.5116 -0.2085>, rotation=<rot 86.0523 0.0004 0.0004>}]
 spine_02 [{shapeType=Sphere, radius=13.2092, center=<-5.1886 1.8975 0.7978>}]
 spine_04 [{shapeType=Capsule, length=10.8718, radius=14.2659, center=<-6.1862 2.3835 -0.015>, rotation=<rot -89.3315 -180 -180>}]
 spine_05 [
  {shapeType=Sphere, radius=8.375, shapeName=RChestFlapCOL, center=<-14.1569 6.3671 8.2573>}
  {shapeType=Capsule, length=14.5017, radius=12.25, shapeName=CChestCapsuleCOL, center=<-8.7406 0.869 0>}
  {shapeType=Capsule, length=5.875, radius=8.875, shapeName=RRibFlapCOL, center=<-20.3784 0.965 6.2726>, rotation=<rot -90 179.9802 -179.9802>}
 ]
 upperarm_r [{shapeType=Capsule, length=9.0232, radius=6.769, center=<-17.6468 1.7618 0.6567>, rotation=<rot 17.4241 92.3096 -84.5915>}]
}
constraints [
 {bone1Name=lowerarm_r, bone2Name=upperarm_r, swing1LimitDegrees=45, swing2LimitDegrees=45, twistLimitDegrees=45}
 {bone1Name=dyn_shawlShoulder_01, bone2Name=dyn_shawlShoulderBase, swing1LimitDegrees=30, swing2LimitDegrees=10, twistLimitDegrees=5}
 {bone1Name=dyn_shawlShoulder_02, bone2Name=dyn_shawlShoulder_01, swing1LimitDegrees=45, swing2LimitDegrees=45, twistLimitDegrees=2}
 {bone1Name=dyn_shawlBackA_01, bone2Name=dyn_shawl_ArmBase, swing1LimitDegrees=20, swing2LimitDegrees=20, twistLimitDegrees=1}
 {bone1Name=dyn_shawlBackA_03, bone2Name=dyn_shawlBackA_01, swing1LimitDegrees=20, swing2LimitDegrees=20, twistLimitDegrees=1}
 {bone1Name=dyn_shawlBackA_02, bone2Name=dyn_shawlBackA_01, swing1LimitDegrees=20, swing2LimitDegrees=20, twistLimitDegrees=1}
 {
  bone1Name=upperarm_r
  bone2Name=spine_05
  swing1LimitDegrees=45
  swing1Motion=Free
  swing2LimitDegrees=45
  swing2Motion=Free
  twistLimitDegrees=45
  twistMotion=Free
 }
 {bone1Name=dyn_shawlFront_01, bone2Name=spine_05, swing1LimitDegrees=20, swing2LimitDegrees=5, twistLimitDegrees=4}
 {
  bone1Name=spine_05
  bone2Name=spine_04
  swing1LimitDegrees=45
  swing1Motion=Free
  swing2LimitDegrees=45
  swing2Motion=Free
  twistLimitDegrees=45
  twistMotion=Free
 }
 {
  bone1Name=spine_04
  bone2Name=spine_02
  swing1LimitDegrees=45
  swing1Motion=Free
  swing2LimitDegrees=45
  swing2Motion=Free
  twistLimitDegrees=45
  twistMotion=Free
 }
 {bone1Name=dyn_backFlapMain, bone2Name=spine_02, swing1LimitDegrees=45, swing2LimitDegrees=10, twistLimitDegrees=20}
 {bone1Name=dyn_backFlap_01, bone2Name=dyn_backFlapMain, swing1LimitDegrees=20, swing2LimitDegrees=10, twistLimitDegrees=5}
 {bone1Name=dyn_backFlapEnd_01, bone2Name=dyn_backFlap_01, swing1LimitDegrees=20, swing2LimitDegrees=10, twistLimitDegrees=5}
 {bone1Name=dyn_backFlap_02, bone2Name=dyn_backFlapMain, swing1LimitDegrees=30, swing2LimitDegrees=10, twistLimitDegrees=5}
 {bone1Name=dyn_backFlap_03, bone2Name=dyn_backFlapMain, swing1LimitDegrees=30, swing2LimitDegrees=10, twistLimitDegrees=5}
 {bone1Name=dyn_shawlShoulderEnd_02, bone2Name=dyn_shawlShoulder_02, swing1LimitDegrees=30, swing2LimitDegrees=10, twistLimitDegrees=10}
 {bone1Name=dyn_shawlShoulderFront_01, bone2Name=dyn_shawlShoulderFrontBase, swing1LimitDegrees=25, swing2LimitDegrees=5, twistLimitDegrees=2}
 {bone1Name=dyn_shawlShoulderFront_02, bone2Name=dyn_shawlShoulderFront_01, swing1LimitDegrees=25, swing2LimitDegrees=5, twistLimitDegrees=2}
 {
  bone1Name=dyn_shawl_ArmBase
  bone2Name=upperarm_r
  swing1LimitDegrees=45
  swing1Motion=Free
  swing2LimitDegrees=45
  swing2Motion=Free
  twistLimitDegrees=45
  twistMotion=Free
 }
 {
  bone1Name=dyn_shawlShoulderBase
  bone2Name=upperarm_r
  swing1LimitDegrees=45
  swing1Motion=Free
  swing2LimitDegrees=45
  swing2Motion=Free
  twistLimitDegrees=45
  twistMotion=Free
 }
 {
  bone1Name=dyn_shawlShoulderFrontBase
  bone2Name=upperarm_r
  swing1LimitDegrees=45
  swing1Motion=Free
  swing2LimitDegrees=45
  swing2Motion=Free
  twistLimitDegrees=45
  twistMotion=Free
 }
]
