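# Animation Blueprint for the male medium Commando Sniper Hood cosmetic body, adding secondary motion to the shawl and back flap.
# It copies the attached parent's pose, runs RigidBody simulation relative to spine_01, then applies weighted multi-bone constraints to the cloth base bones.
# LodLow disables the arm and shoulder constraints above LOD 1, while front shawl and back flap constraints stay on.
# The shoulder-front constraint appears twice in succession with identical settings.
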AnimBlueprint /BRCosmetics/Characters/Player/Male/Medium/Bodies/M_MED_Commando_SniperHood/Meshes/M_MED_Sniper_Hood_Body_MESH_Skeleton_AnimBlueprint
target_skeleton=/Game/Characters/Player/Male/Medium/Bodies/M_MED_Commando_SniperHood/Meshes/M_MED_Sniper_Hood_Body_MESH_Skeleton
graphs [
 AnimGraph {nodes [
   n1 = RigidBody(ComponentPose=Local To Component(LocalPose=Copy Pose From Mesh() {bUseAttachedParent=true})) {
    bOverrideWorldGravity=true
    overridePhysicsAsset=&/M_MED_Sniper_Hood_Physics
    simulationSpace=BaseBoneSpace
    alphaBoolBlend {blendOption=Linear}
    baseBoneRef {boneName=spine_01}
    componentAppliedLinearAccClamp=<0.1 0.1 0.1>
    componentLinearAccScale=<0.1 0.1 0.1>
    componentLinearVelScale=<0.8 0.8 0.8>
    overrideWorldGravity {z=-980}
   }
   n2 = "Constraint - dyn_shawlFrontBase (PP)"!(ComponentPose=@n1, Alpha=1, ConstraintWeights_0=0.8, ConstraintWeights_1:float=0.2) -> (Pose) {
    alphaBoolBlend {blendOption=Linear}
    boneToModify {boneName=dyn_shawlFrontBase}
    constraintSetup [{targetBone {boneName=clavicle_r}}, {targetBone {boneName=spine_05}}]
    constraintWeights [1,1]
   }
   n3 = Get LodLow() -> (LodLow:double)
   n4 = "Constraint - dyn_shawl_ArmBase (PPPP)"!(
    ComponentPose=@n2
    Alpha=@n3
    ConstraintWeights_0=0.1
    ConstraintWeights_1:float=0.4
    ConstraintWeights_2:float=0.4
    ConstraintWeights_3:float=0.1
   ) -> (Pose) {
    alphaBoolBlend {blendOption=Linear}
    boneToModify {boneName=dyn_shawl_ArmBase}
    constraintSetup [
     {targetBone {boneName=spine_03}}
     {targetBone {boneName=clavicle_r}}
     {targetBone {boneName=upperarm_r}}
     {targetBone {boneName=upperarm_twist_01_r}}
    ]
    constraintWeights [1,1,1,1]
   }
   n5 = "Constraint - dyn_shawlShoulderBase (PPPP)"!(
    ComponentPose=@n4
    Alpha=@n3
    ConstraintWeights_0=0.45
    ConstraintWeights_1:float=0.2
    ConstraintWeights_2:float=0.3
    ConstraintWeights_3:float=0.05
   ) -> (Pose) {
    alphaBoolBlend {blendOption=Linear}
    boneToModify {boneName=dyn_shawlShoulderBase}
    constraintSetup [
     {targetBone {boneName=upperarm_twist_01_r}}
     {targetBone {boneName=upperarm_twist_02_r}}
     {targetBone {boneName=upperarm_r}}
     {targetBone {boneName=clavicle_r}}
    ]
    constraintWeights [1,1,1,1]
   }
   n6 = "Constraint - dyn_backFlapBase (PPPPPPP)"!(
    ComponentPose=@n5
    Alpha=1
    ConstraintWeights_0=0.05
    ConstraintWeights_1:float=0.1
    ConstraintWeights_2:float=0.1
    ConstraintWeights_3:float=0.5
    ConstraintWeights_4:float=0.075
    ConstraintWeights_5:float=0.075
    ConstraintWeights_6:float=0.1
   ) -> (Pose) {
    alphaBoolBlend {blendOption=Linear}
    boneToModify {boneName=dyn_backFlapBase}
    constraintSetup [
     {targetBone {boneName=pelvis}}
     {targetBone {boneName=spine_01}}
     {targetBone {boneName=spine_02}}
     {targetBone {boneName=spine_03}}
     {targetBone {boneName=clavicle_l}}
     {targetBone {boneName=clavicle_r}}
     {targetBone {boneName=spine_04}}
    ]
    constraintWeights [1,1,1,1,1,1,1]
   }
   n7 = "Constraint - dyn_shawlShoulderFrontBase (PPPPPP)"!(
    ComponentPose=@n6
    Alpha=@n3
    ConstraintWeights_0=0.05
    ConstraintWeights_1:float=0.075
    ConstraintWeights_2:float=0.125
    ConstraintWeights_3:float=0.125
    ConstraintWeights_4:float=0.025
    ConstraintWeights_5:float=0.6
   ) -> (Pose) {
    alphaBoolBlend {blendOption=Linear}
    boneToModify {boneName=dyn_shawlShoulderFrontBase}
    constraintSetup [
     {targetBone {boneName=clavicle_r}}
     {targetBone {boneName=dyn_shawlShoulder_01}}
     {targetBone {boneName=lowerarm_r}}
     {targetBone {boneName=lowerarm_twist_02_r}}
     {targetBone {boneName=spine_03}}
     {targetBone {boneName=upperarm_r}}
    ]
    constraintWeights [1,1,1,1,1,1]
   }
   n8 = "Constraint - dyn_shawlShoulderFrontBase (PPPPPP)"(
    ComponentPose=@n7
    Alpha=@n3
    ConstraintWeights_0=0.05
    ConstraintWeights_1=0.075
    ConstraintWeights_2=0.125
    ConstraintWeights_3=0.125
    ConstraintWeights_4=0.025
    ConstraintWeights_5=0.6
   ) {
    alphaBoolBlend {blendOption=Linear}
    boneToModify {boneName=dyn_shawlShoulderFrontBase}
    constraintSetup [
     {targetBone {boneName=clavicle_r}}
     {targetBone {boneName=dyn_shawlShoulder_01}}
     {targetBone {boneName=lowerarm_r}}
     {targetBone {boneName=lowerarm_twist_02_r}}
     {targetBone {boneName=spine_03}}
     {targetBone {boneName=upperarm_r}}
    ]
    constraintWeights [1,1,1,1,1,1]
   }
   Output Pose(Result=Component To Local(ComponentPose=@n8)) {layerGroup=None}
  ]}
 EventGraph {nodes [
   Event BlueprintUpdateAnimation()
   n1 = Get CurrentLOD()
   > Set LodLow(LodLow:double=SelectFloat(A=0, B=1, bPickA="integer > integer"(A=@n1, B=1) {is=title K2Node_CallFunction})) -> (Output_Get:double)
   TryGetPawnOwner()
  ]}
]
variables [LodLow {type=number}]
